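# Test NPC conversation database for the Tandem system, entered through Conversation.Entry.TandemTest, mainly exercising the outfit disguise service.
# A greeting Speech node picks contextual lines for first-ever conversations or a specific Lancelot outfit, falling back to a default.
# It branches to a disguise choice gated on the OutfitDisguise service, leading to Confirm or Back, with a per-entry-tag speech override.
# A second branch links to the remote Season38 SilentNovel conversation entry.
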ConversationDatabase /NPCLibrary/NPCs/_TEST_NPC/NPCLibrary_Conversation_TandemTest
entry_points [{tag=Conversation.Entry.TandemTest, destinations [#1]}]
nodes [
 ConversationEntryPointNode_7 {
  id=#1
  class=ConversationEntryPointNode
  outputs [#2]
  properties {entryTag {tagName=Conversation.Entry.TandemTest}}
 }
 Speech_C_4 {
  id=#2
  class=Speech_C
  outputs [#3, #4]
  properties {
   nodeName=""
   generalConfig {
    defaultMessage=Default conversation.
    contextualMessages [
     {
      message=First ever conversation.
      contextRequirements [{requirement {class=FortControllerRequirement_IsFirstEverConversationWithNPC}}]
      priority {curve {rowName=Tandem.Speech.Priority.FirstEverInteraction}}
     }
     {
      message=Outfit reaction.
      contextRequirements [{requirement {
         class=FortControllerRequirement_HasCID
         properties {allowedCIds [/BRCosmetics/Athena/Items/Cosmetics/Characters/CID_A_424_Athena_Commando_M_Lancelot]}
        }}]
      priority {value=0, curve {curveTable=null, rowName=None}}
     }
    ]
   }
  }
 }
 ConversationLinkNode_0 {
  id=#4
  class=ConversationLinkNode
  properties {remoteEntryTag {tagName=Conversation.Entry.Season38.SilentNovel}}
 }
 Speech_C_0 {
  id=#3
  class=Speech_C
  outputs [#5, #6]
  properties {
   bUsePerNPC_ContextualMessageLookup=true
   softTaskIcon=/Game/Athena/HUD/Interaction/T_UI_NPC_Intel
   generalConfig {defaultMessage="They'll never recognize you!"}
   speakerEntryTagToConfig {"(TagName=\"Conversation.Entry.TandemTest\")" {defaultMessage="You're a master of disguise!"}}
  }
  sub_nodes [
   FortConversationChoiceWithPreviewNode_2 {class=FortConversationChoiceWithPreviewNode, properties {defaultChoiceDisplayText=OUTFIT DISGUISE TEST}}
   FortConversationRequirement_HasService_2 {class=FortConversationRequirement_HasService, properties {serviceTag {tagName=Tandem.Service.OutfitDisguise}}}
  ]
 }
 OutfitDisguise_C_3 {
  id=#6
  class=OutfitDisguise_C
  outputs []
  sub_nodes [ConversationChoiceNode_7 {class=ConversationChoiceNode, properties {defaultChoiceDisplayText=CONFIRM}}]
 }
 Back_C_6 {
  id=#5
  class=Back_C
  sub_nodes [ConversationChoiceNode_6 {class=ConversationChoiceNode, properties {defaultChoiceDisplayText=BACK}}]
 }
]
